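# Physics asset for the skydive variant of the CinderMax female medium body, covering its dangling chain, strap, grenade and shoe flaps.
# Capsules on pelvis, spine_04 and both thighs form a coarse body for these dyn_ bones.
# Small base capsules anchor chain and strap to pelvis and flaps to thighs with free angular motion; the grenade joins pelvis directly.
# Accessory joints mostly limit swing and tightly limit twist, but the first two chain links swing freely.
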PhysicsAsset /BRCosmetics/Characters/Player/Female/Medium/Bodies/F_MED_CinderMax/Meshes/F_MED_CinderMax_Physics_skydive
bodies {
 dyn_chain_01 [{
   shapeType=Capsule
   length=6.0756
   radius=1.5
   shapeName=dyn_chain_01_capsule
   center=<4.1877 -0.0376 0.5229>
   rotation=<rot 0 88.2277 -86.5524>
  }]
 dyn_chain_01_base [{shapeType=Capsule, length=0.505, radius=0.505, shapeName=dyn_chain_01_base_capsule, rotation=<rot 0 -90 0>}]
 dyn_chain_02 [{shapeType=Capsule, length=5.0727, radius=1, shapeName=dyn_chain_02_capsule, center=<4.8908 -0.0145 0.814>, rotation=<rot 0 84.8272 -88.984>}]
 dyn_chain_03 [{shapeType=Capsule, length=4.5325, radius=1, shapeName=dyn_chain_03_capsule, center=<4.7934 -0.641 0.3552>, rotation=<rot 0 -95.178 -88.318>}]
 dyn_chain_04 [{
   shapeType=Capsule
   length=4.5078
   radius=1.5
   shapeName=dyn_chain_04_capsule
   center=<4.8351 -0.5786 -0.0987>
   rotation=<rot 0 -92.1982 -88.1022>
  }]
 dyn_chain_05 [{
   shapeType=Capsule
   length=4.3015
   radius=1.5
   shapeName=dyn_chain_05_capsule
   center=<2.9001 -0.911 0.7809>
   rotation=<rot 0.6877 -98.996 -86.0965>
  }]
 dyn_chain_06 [{shapeType=Capsule, length=0.505, radius=0.505, shapeName=dyn_chain_06_capsule, rotation=<rot 0 -90 0>}]
 dyn_grenade_01 [{
   shapeType=Capsule
   length=7.75
   radius=3.9563
   shapeName=dyn_grenade_01_capsule
   center=<5.095 0.0748 2.3057>
   rotation=<rot 0 -87.6767 -89.1809>
  }]
 dyn_shoe_flaps_l [{
   shapeType=Capsule
   length=6.0589
   radius=1.9913
   shapeName=dyn_shoe_flaps_l_capsule
   center=<6.1168 0 0>
   rotation=<rot -90 144.7356 -144.7358>
  }]
 dyn_shoe_flaps_l_base [{shapeType=Capsule, length=0.505, radius=0.505, shapeName=dyn_shoe_flaps_l_base_capsule, rotation=<rot 0 -90 0>}]
 dyn_shoe_flaps_r [{shapeType=Capsule, length=6.0589, radius=1.9914, shapeName=dyn_shoe_flaps_r_capsule, center=<6.1168 0 0>, rotation=<rot -90 0 -0.0002>}]
 dyn_shoe_flaps_r_base [{shapeType=Capsule, length=0.505, radius=0.505, shapeName=dyn_shoe_flaps_r_base_capsule, rotation=<rot 0 -90 0>}]
 dyn_strap_01 [{shapeType=Capsule, length=4, radius=2, shapeName=dyn_strap_01_capsule, center=<3.1874 0 1.4785>, rotation=<rot 80 -179.9997 -180>}]
 dyn_strap_01_base [{shapeType=Capsule, length=0.505, radius=0.505, shapeName=dyn_strap_01_base_capsule, rotation=<rot 0 -90 0>}]
 dyn_strap_02 [
  {shapeType=Capsule, length=3.9801, radius=1.0033, shapeName=dyn_strap_02_capsule, center=<3.5836 0 0>, rotation=<rot 90 -179.0505 -179.0507>}
  {shapeType=Capsule, length=3.5, radius=1, shapeName=dyn_strap_02_capsule, center=<3.492 -2.6018 0>, rotation=<rot 0 -89.9996 -90>}
 ]
 dyn_strap_03 [{shapeType=Capsule, length=3.9801, radius=1.0033, shapeName=dyn_strap_03_capsule, center=<3.5836 0 0>, rotation=<rot 90 -90 -90.0001>}]
 dyn_strap_04 [{shapeType=Capsule, length=3.9801, radius=1.0033, shapeName=dyn_strap_04_capsule, center=<3.5836 0 0>, rotation=<rot 90 -179.7287 -179.7288>}]
 dyn_strap_05 [{shapeType=Capsule, length=3.9801, radius=1.0033, shapeName=dyn_strap_05_capsule, center=<3.5836 0 0>, rotation=<rot 90 -0.1186 -0.1187>}]
 dyn_strap_06 [{shapeType=Capsule, length=3.9801, radius=1.0033, shapeName=dyn_strap_06_capsule, center=<3.5836 0 0>, rotation=<rot 90 -179.9901 -179.9902>}]
 pelvis [
  {shapeType=Capsule, length=2.8085, radius=9.9495, shapeName=pelvis_L, center=<-1.3974 0 -5.9558>, rotation=<rot -80 -0.0001 0>}
  {shapeType=Capsule, length=11.6369, radius=8.1024, shapeName=pelvis_R, center=<0.1832 -2.2368 8.5492>, rotation=<rot -90 -180 179.9999>}
 ]
 spine_04 [{shapeType=Capsule, length=12.7012, radius=5.9675, center=<-0.843 1.1079 -0.1752>, rotation=<rot 2.1651 -90 -0.9327>}]
 thigh_l [{shapeType=Capsule, length=16.2153, radius=7.5849, shapeName=thigh_r_capsule, center=<-16.1716 0 0>, rotation=<rot 90 173.2786 -6.7213>}]
 thigh_r [{shapeType=Capsule, length=16.2153, radius=7.5849, shapeName=thigh_r_capsule, center=<16.1716 0 0>, rotation=<rot 90 89.4328 89.4328>}]
}
constraints [
 {bone1Name=spine_04, bone2Name=pelvis, swing1LimitDegrees=45, swing2LimitDegrees=45, twistLimitDegrees=45}
 {bone1Name=thigh_r, bone2Name=pelvis, swing1LimitDegrees=45, swing2LimitDegrees=45, twistLimitDegrees=45}
 {bone1Name=thigh_l, bone2Name=pelvis, swing1LimitDegrees=45, swing2LimitDegrees=45, twistLimitDegrees=45}
 {bone1Name=dyn_grenade_01, bone2Name=pelvis, swing1LimitDegrees=45, swing2LimitDegrees=45, twistLimitDegrees=2}
 {bone1Name=dyn_strap_01, bone2Name=dyn_strap_01_base, swing1LimitDegrees=15, swing2LimitDegrees=15, twistLimitDegrees=2}
 {bone1Name=dyn_strap_02, bone2Name=dyn_strap_01, swing1LimitDegrees=45, swing2LimitDegrees=45, twistLimitDegrees=2}
 {bone1Name=dyn_strap_03, bone2Name=dyn_strap_02, swing1LimitDegrees=45, swing2LimitDegrees=45, twistLimitDegrees=2}
 {bone1Name=dyn_strap_04, bone2Name=dyn_strap_03, swing1LimitDegrees=45, swing2LimitDegrees=45, twistLimitDegrees=2}
 {bone1Name=dyn_strap_05, bone2Name=dyn_strap_04, swing1LimitDegrees=45, swing2LimitDegrees=45, twistLimitDegrees=2}
 {bone1Name=dyn_strap_06, bone2Name=dyn_strap_05, swing1LimitDegrees=45, swing2LimitDegrees=45, twistLimitDegrees=2}
 {bone1Name=dyn_shoe_flaps_r, bone2Name=dyn_shoe_flaps_r_base, swing1LimitDegrees=30, swing2LimitDegrees=15, twistLimitDegrees=2}
 {
  bone1Name=dyn_shoe_flaps_l_base
  bone2Name=thigh_l
  swing1LimitDegrees=45
  swing1Motion=Free
  swing2LimitDegrees=45
  swing2Motion=Free
  twistLimitDegrees=45
  twistMotion=Free
 }
 {bone1Name=dyn_shoe_flaps_l, bone2Name=dyn_shoe_flaps_l_base, swing1LimitDegrees=30, swing2LimitDegrees=15, twistLimitDegrees=2}
 {
  bone1Name=dyn_shoe_flaps_r_base
  bone2Name=thigh_r
  swing1LimitDegrees=45
  swing1Motion=Free
  swing2LimitDegrees=45
  swing2Motion=Free
  twistLimitDegrees=45
  twistMotion=Free
 }
 {
  bone1Name=dyn_chain_01_base
  bone2Name=pelvis
  swing1LimitDegrees=45
  swing1Motion=Free
  swing2LimitDegrees=45
  swing2Motion=Free
  twistLimitDegrees=45
  twistMotion=Free
 }
 {
  bone1Name=dyn_chain_01
  bone2Name=dyn_chain_01_base
  swing1LimitDegrees=60
  swing1Motion=Free
  swing2LimitDegrees=60
  swing2Motion=Free
  twistLimitDegrees=10
 }
 {
  bone1Name=dyn_chain_02
  bone2Name=dyn_chain_01
  swing1LimitDegrees=60
  swing1Motion=Free
  swing2LimitDegrees=60
  swing2Motion=Free
  twistLimitDegrees=10
 }
 {bone1Name=dyn_chain_03, bone2Name=dyn_chain_02, swing1LimitDegrees=60, swing2LimitDegrees=60, twistLimitDegrees=10}
 {bone1Name=dyn_chain_04, bone2Name=dyn_chain_03, swing1LimitDegrees=60, swing2LimitDegrees=60, twistLimitDegrees=10}
 {bone1Name=dyn_chain_05, bone2Name=dyn_chain_04, swing1LimitDegrees=60, swing2LimitDegrees=60, twistLimitDegrees=10}
 {bone1Name=dyn_chain_06, bone2Name=dyn_chain_05, swing1LimitDegrees=90, swing2LimitDegrees=90, twistLimitDegrees=2}
 {
  bone1Name=dyn_strap_01_base
  bone2Name=pelvis
  swing1LimitDegrees=45
  swing1Motion=Free
  swing2LimitDegrees=45
  swing2Motion=Free
  twistLimitDegrees=45
  twistMotion=Free
 }
]
