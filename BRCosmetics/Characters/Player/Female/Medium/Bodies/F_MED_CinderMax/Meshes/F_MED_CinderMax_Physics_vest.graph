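# Physics asset for the F_MED_CinderMax female body's vest, providing secondary motion for its bottom hem, fur tufts and zippers.
# Each strand has a base capsule constrained to the pelvis with free swing and twist, driving limited child capsules.
# Back hem strands are two-link chains with tight limits, other strands single links, with wider swing on fur and zippers.
# Three pelvis capsules form the only body volume.
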PhysicsAsset /BRCosmetics/Characters/Player/Female/Medium/Bodies/F_MED_CinderMax/Meshes/F_MED_CinderMax_Physics_vest
bodies {
 dyn_vest_bottom_bk_01 [{shapeType=Capsule, length=3, radius=2, shapeName=dyn_vest_bottom_bk_01_capsule, center=<3.3076 0 0>, rotation=<rot 0 0.0004 0>}]
 dyn_vest_bottom_bk_02 [{shapeType=Capsule, length=3, radius=2, shapeName=dyn_vest_bottom_bk_02_capsule, center=<3.3076 0 0>, rotation=<rot 0 0.0008 0>}]
 dyn_vest_bottom_bk_base [{shapeType=Capsule, length=0.505, radius=0.505, shapeName=dyn_vest_bottom_bk_base_capsule, rotation=<rot 0 -90 0>}]
 dyn_vest_bottom_bk_side_l_01 [{shapeType=Capsule, length=3, radius=2, shapeName=dyn_vest_bottom_bk_side_l_01_capsule, center=<3.3076 0 0>, rotation=<rot 0.0001 0.0003 0>}]
 dyn_vest_bottom_bk_side_l_02 [{shapeType=Capsule, length=3, radius=2, shapeName=dyn_vest_bottom_bk_side_l_02_capsule, center=<3.3076 0 0>, rotation=<rot 0.0002 0.0007 0>}]
 dyn_vest_bottom_bk_side_l_base [{shapeType=Capsule, length=0.505, radius=0.505, shapeName=dyn_vest_bottom_bk_side_l_base_capsule, rotation=<rot 0 -90 0>}]
 dyn_vest_bottom_bk_side_r_01 [{shapeType=Capsule, length=3, radius=2, shapeName=dyn_vest_bottom_bk_side_r_01_capsule, center=<3.3076 0 0>, rotation=<rot -0.0001 0.0003 0>}]
 dyn_vest_bottom_bk_side_r_02 [{
   shapeType=Capsule
   length=3
   radius=2
   shapeName=dyn_vest_bottom_bk_side_r_02_capsule
   center=<3.3076 0 0>
   rotation=<rot -0.0002 0.0007 0.0001>
  }]
 dyn_vest_bottom_bk_side_r_base [{shapeType=Capsule, length=0.505, radius=0.505, shapeName=dyn_vest_bottom_bk_side_r_base_capsule, rotation=<rot 0 -90 0>}]
 dyn_vest_bottom_fr_l [{shapeType=Capsule, length=3, radius=2, shapeName=dyn_vest_bottom_fr_l_capsule, center=<3.8276 0 0>}]
 dyn_vest_bottom_fr_l_base [{shapeType=Capsule, length=0.505, radius=0.505, shapeName=dyn_vest_bottom_fr_l_base_capsule, rotation=<rot 0 -90 0>}]
 dyn_vest_bottom_fr_r [{shapeType=Capsule, length=3, radius=2, shapeName=dyn_vest_bottom_fr_r_capsule, center=<3.8276 0 0>}]
 dyn_vest_bottom_fr_r_base [{shapeType=Capsule, length=0.505, radius=0.505, shapeName=dyn_vest_bottom_fr_r_base_capsule, rotation=<rot 0 -90 0>}]
 dyn_vest_bottom_fr_side_l [{shapeType=Capsule, length=3, radius=2, shapeName=dyn_vest_bottom_fr_side_l_capsule, center=<3.8276 0 0>}]
 dyn_vest_bottom_fr_side_l_base [{shapeType=Capsule, length=0.505, radius=0.505, shapeName=dyn_vest_bottom_fr_side_l_base_capsule, rotation=<rot 0 -90 0>}]
 dyn_vest_bottom_fr_side_r [{shapeType=Capsule, length=3, radius=2, shapeName=dyn_vest_bottom_fr_side_r_capsule, center=<3.8276 0 0>}]
 dyn_vest_bottom_fr_side_r_base [{shapeType=Capsule, length=0.505, radius=0.505, shapeName=dyn_vest_bottom_fr_side_r_base_capsule, rotation=<rot 0 -90 0>}]
 dyn_vest_bottom_side_l [{shapeType=Capsule, length=3, radius=2, shapeName=dyn_vest_bottom_side_l_capsule, center=<3.8276 0 0>}]
 dyn_vest_bottom_side_l_base [{shapeType=Capsule, length=0.505, radius=0.505, shapeName=dyn_vest_bottom_side_l_base_capsule, rotation=<rot 0 -90 0>}]
 dyn_vest_bottom_side_r [{shapeType=Capsule, length=3, radius=2, shapeName=dyn_vest_bottom_side_r_capsule, center=<3.8276 0 0>}]
 dyn_vest_bottom_side_r_base [{shapeType=Capsule, length=0.505, radius=0.505, shapeName=dyn_vest_bottom_side_r_base_capsule, rotation=<rot 0 -90 0>}]
 dyn_vest_fur_01_l [{shapeType=Capsule, length=2, radius=1, shapeName=dyn_vest_fur_01_r_capsule, center=<1.2838 0 0>, rotation=<rot 0 -0.0001 90>}]
 dyn_vest_fur_01_l_base [{shapeType=Capsule, length=0.505, radius=0.505, shapeName=dyn_vest_fur_01_r_base_capsule, rotation=<rot 90 90 -0.0002>}]
 dyn_vest_fur_01_r [{shapeType=Capsule, length=2, radius=1, shapeName=dyn_vest_fur_01_r_capsule, center=<1.1266 0 0>, rotation=<rot 0 0.0001 -89.9999>}]
 dyn_vest_fur_01_r_base [{shapeType=Capsule, length=0.505, radius=0.505, shapeName=dyn_vest_fur_01_r_base_capsule, rotation=<rot 90 112.1698 -157.8304>}]
 dyn_vest_fur_02_l [{shapeType=Capsule, length=2, radius=1, shapeName=dyn_vest_fur_02_r_capsule, center=<1.2838 0 0>, rotation=<rot 0 -0.0001 90>}]
 dyn_vest_fur_02_l_base [{shapeType=Capsule, length=0.505, radius=0.505, shapeName=dyn_vest_fur_02_r_base_capsule, rotation=<rot 90 90 -0.0003>}]
 dyn_vest_fur_02_r [{shapeType=Capsule, length=2, radius=1, shapeName=dyn_vest_fur_02_r_capsule, center=<1.1266 0 0>, rotation=<rot 0 0.0001 -90>}]
 dyn_vest_fur_02_r_base [{shapeType=Capsule, length=0.505, radius=0.505, shapeName=dyn_vest_fur_02_r_base_capsule, rotation=<rot 90 62.3683 152.3681>}]
 dyn_vest_fur_03_l [{shapeType=Capsule, length=2, radius=1, shapeName=dyn_vest_fur_03_r_capsule, center=<1.2838 0 0>, rotation=<rot 0 -0.0001 90>}]
 dyn_vest_fur_03_l_base [{shapeType=Capsule, length=0.505, radius=0.505, shapeName=dyn_vest_fur_03_r_base_capsule, rotation=<rot 90 90 -0.0001>}]
 dyn_vest_fur_03_r [{shapeType=Capsule, length=2, radius=1, shapeName=dyn_vest_fur_03_r_capsule, center=<0.9851 0 0>, rotation=<rot 90 153.4349 -116.5651>}]
 dyn_vest_fur_03_r_base [{shapeType=Capsule, length=0.505, radius=0.505, shapeName=dyn_vest_fur_03_r_base_capsule, rotation=<rot 90 64.7217 154.7216>}]
 dyn_zipper_l [{
   shapeType=Capsule
   length=0.505
   radius=0.505
   shapeName=dyn_zipper_l_capsule
   center=<1.5326 -0.001 -0.0376>
   rotation=<rot 0 -90.0122 -88.9855>
  }]
 dyn_zipper_l_MSTR [{shapeType=Capsule, length=0.505, radius=0.505, shapeName=dyn_zipper_l_MSTR_capsule, rotation=<rot 0 -90 0>}]
 dyn_zipper_r [{
   shapeType=Capsule
   length=0.505
   radius=0.505
   shapeName=dyn_zipper_r_capsule
   center=<1.5326 0.001 -0.0375>
   rotation=<rot 0 -89.9878 -88.9855>
  }]
 dyn_zipper_r_MSTR [{shapeType=Capsule, length=0.505, radius=0.505, shapeName=dyn_zipper_r_MSTR_capsule, rotation=<rot 0 -90 0>}]
 pelvis [
  {shapeType=Capsule, length=21.3482, radius=12.2379}
  {shapeType=Capsule, length=6.5, radius=7.5, shapeName=pelvis_capsule, center=<15.9139 -0.4729 0>}
  {shapeType=Capsule, length=6, radius=7.5, shapeName=pelvis_capsule, center=<13.9471 2.5006 0>}
 ]
}
constraints [
 {
  bone1Name=dyn_vest_bottom_bk_base
  bone2Name=pelvis
  swing1LimitDegrees=45
  swing1Motion=Free
  swing2LimitDegrees=45
  swing2Motion=Free
  twistLimitDegrees=45
  twistMotion=Free
 }
 {bone1Name=dyn_vest_bottom_bk_01, bone2Name=dyn_vest_bottom_bk_base, swing1LimitDegrees=5, swing2LimitDegrees=2, twistLimitDegrees=2}
 {bone1Name=dyn_vest_bottom_bk_02, bone2Name=dyn_vest_bottom_bk_01, swing1LimitDegrees=5, swing2LimitDegrees=2, twistLimitDegrees=2}
 {
  bone1Name=dyn_vest_bottom_bk_side_l_base
  bone2Name=pelvis
  swing1LimitDegrees=45
  swing1Motion=Free
  swing2LimitDegrees=45
  swing2Motion=Free
  twistLimitDegrees=45
  twistMotion=Free
 }
 {
  bone1Name=dyn_vest_bottom_bk_side_l_01
  bone2Name=dyn_vest_bottom_bk_side_l_base
  swing1LimitDegrees=5
  swing2LimitDegrees=2
  twistLimitDegrees=2
 }
 {
  bone1Name=dyn_vest_bottom_bk_side_l_02
  bone2Name=dyn_vest_bottom_bk_side_l_01
  swing1LimitDegrees=5
  swing2LimitDegrees=2
  twistLimitDegrees=2
 }
 {
  bone1Name=dyn_vest_bottom_bk_side_r_base
  bone2Name=pelvis
  swing1LimitDegrees=45
  swing1Motion=Free
  swing2LimitDegrees=45
  swing2Motion=Free
  twistLimitDegrees=45
  twistMotion=Free
 }
 {
  bone1Name=dyn_vest_bottom_bk_side_r_01
  bone2Name=dyn_vest_bottom_bk_side_r_base
  swing1LimitDegrees=5
  swing2LimitDegrees=2
  twistLimitDegrees=2
 }
 {
  bone1Name=dyn_vest_bottom_bk_side_r_02
  bone2Name=dyn_vest_bottom_bk_side_r_01
  swing1LimitDegrees=5
  swing2LimitDegrees=2
  twistLimitDegrees=2
 }
 {
  bone1Name=dyn_vest_fur_03_r_base
  bone2Name=pelvis
  swing1LimitDegrees=45
  swing1Motion=Free
  swing2LimitDegrees=45
  swing2Motion=Free
  twistLimitDegrees=45
  twistMotion=Free
 }
 {bone1Name=dyn_vest_fur_03_r, bone2Name=dyn_vest_fur_03_r_base, swing1LimitDegrees=45, swing2LimitDegrees=45, twistLimitDegrees=2}
 {
  bone1Name=dyn_vest_fur_02_r_base
  bone2Name=pelvis
  swing1LimitDegrees=45
  swing1Motion=Free
  swing2LimitDegrees=45
  swing2Motion=Free
  twistLimitDegrees=45
  twistMotion=Free
 }
 {
  bone1Name=dyn_vest_fur_01_r_base
  bone2Name=pelvis
  swing1LimitDegrees=45
  swing1Motion=Free
  swing2LimitDegrees=45
  swing2Motion=Free
  twistLimitDegrees=45
  twistMotion=Free
 }
 {bone1Name=dyn_vest_fur_01_r, bone2Name=dyn_vest_fur_01_r_base, swing1LimitDegrees=45, swing2LimitDegrees=45, twistLimitDegrees=2}
 {bone1Name=dyn_vest_fur_02_r, bone2Name=dyn_vest_fur_02_r_base, swing1LimitDegrees=45, swing2LimitDegrees=45, twistLimitDegrees=2}
 {
  bone1Name=dyn_vest_fur_01_l_base
  bone2Name=pelvis
  swing1LimitDegrees=45
  swing1Motion=Free
  swing2LimitDegrees=45
  swing2Motion=Free
  twistLimitDegrees=45
  twistMotion=Free
 }
 {bone1Name=dyn_vest_fur_01_l, bone2Name=dyn_vest_fur_01_l_base, swing1LimitDegrees=45, swing2LimitDegrees=45, twistLimitDegrees=2}
 {
  bone1Name=dyn_vest_fur_02_l_base
  bone2Name=pelvis
  swing1LimitDegrees=45
  swing1Motion=Free
  swing2LimitDegrees=45
  swing2Motion=Free
  twistLimitDegrees=45
  twistMotion=Free
 }
 {bone1Name=dyn_vest_fur_02_l, bone2Name=dyn_vest_fur_02_l_base, swing1LimitDegrees=45, swing2LimitDegrees=45, twistLimitDegrees=2}
 {
  bone1Name=dyn_vest_fur_03_l_base
  bone2Name=pelvis
  swing1LimitDegrees=45
  swing1Motion=Free
  swing2LimitDegrees=45
  swing2Motion=Free
  twistLimitDegrees=45
  twistMotion=Free
 }
 {bone1Name=dyn_vest_fur_03_l, bone2Name=dyn_vest_fur_03_l_base, swing1LimitDegrees=45, swing2LimitDegrees=45, twistLimitDegrees=2}
 {bone1Name=dyn_zipper_r, bone2Name=dyn_zipper_r_MSTR, swing1LimitDegrees=100, swing2LimitDegrees=50, twistLimitDegrees=2}
 {
  bone1Name=dyn_zipper_l_MSTR
  bone2Name=pelvis
  swing1LimitDegrees=45
  swing1Motion=Free
  swing2LimitDegrees=45
  swing2Motion=Free
  twistLimitDegrees=45
  twistMotion=Free
 }
 {bone1Name=dyn_zipper_l, bone2Name=dyn_zipper_l_MSTR, swing1LimitDegrees=100, swing2LimitDegrees=50, twistLimitDegrees=2}
 {
  bone1Name=dyn_zipper_r_MSTR
  bone2Name=pelvis
  swing1LimitDegrees=45
  swing1Motion=Free
  swing2LimitDegrees=45
  swing2Motion=Free
  twistLimitDegrees=45
  twistMotion=Free
 }
 {bone1Name=dyn_vest_bottom_side_r, bone2Name=dyn_vest_bottom_side_r_base, swing1LimitDegrees=10, swing2LimitDegrees=2, twistLimitDegrees=2}
 {
  bone1Name=dyn_vest_bottom_fr_l_base
  bone2Name=pelvis
  swing1LimitDegrees=45
  swing1Motion=Free
  swing2LimitDegrees=45
  swing2Motion=Free
  twistLimitDegrees=45
  twistMotion=Free
 }
 {bone1Name=dyn_vest_bottom_fr_l, bone2Name=dyn_vest_bottom_fr_l_base, swing1LimitDegrees=50, swing2LimitDegrees=2, twistLimitDegrees=2}
 {
  bone1Name=dyn_vest_bottom_fr_r_base
  bone2Name=pelvis
  swing1LimitDegrees=45
  swing1Motion=Free
  swing2LimitDegrees=45
  swing2Motion=Free
  twistLimitDegrees=45
  twistMotion=Free
 }
 {bone1Name=dyn_vest_bottom_fr_r, bone2Name=dyn_vest_bottom_fr_r_base, swing1LimitDegrees=50, swing2LimitDegrees=2, twistLimitDegrees=2}
 {
  bone1Name=dyn_vest_bottom_fr_side_l_base
  bone2Name=pelvis
  swing1LimitDegrees=45
  swing1Motion=Free
  swing2LimitDegrees=45
  swing2Motion=Free
  twistLimitDegrees=45
  twistMotion=Free
 }
 {
  bone1Name=dyn_vest_bottom_fr_side_l
  bone2Name=dyn_vest_bottom_fr_side_l_base
  swing1LimitDegrees=15
  swing2LimitDegrees=2
  twistLimitDegrees=2
 }
 {
  bone1Name=dyn_vest_bottom_fr_side_r_base
  bone2Name=pelvis
  swing1LimitDegrees=45
  swing1Motion=Free
  swing2LimitDegrees=45
  swing2Motion=Free
  twistLimitDegrees=45
  twistMotion=Free
 }
 {
  bone1Name=dyn_vest_bottom_fr_side_r
  bone2Name=dyn_vest_bottom_fr_side_r_base
  swing1LimitDegrees=15
  swing2LimitDegrees=2
  twistLimitDegrees=2
 }
 {
  bone1Name=dyn_vest_bottom_side_l_base
  bone2Name=pelvis
  swing1LimitDegrees=45
  swing1Motion=Free
  swing2LimitDegrees=45
  swing2Motion=Free
  twistLimitDegrees=45
  twistMotion=Free
 }
 {bone1Name=dyn_vest_bottom_side_l, bone2Name=dyn_vest_bottom_side_l_base, swing1LimitDegrees=10, swing2LimitDegrees=2, twistLimitDegrees=2}
 {
  bone1Name=dyn_vest_bottom_side_r_base
  bone2Name=pelvis
  swing1LimitDegrees=45
  swing1Motion=Free
  swing2LimitDegrees=45
  swing2Motion=Free
  twistLimitDegrees=45
  twistMotion=Free
 }
]
