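# Animation Blueprint for the CinderMax female body, adding secondary motion to chain, strap, vest and other costume bones.
# The parent mesh pose passes through a deform Control Rig and weighted constraints into a cached pose, feeding bool-blended standard, crouch and skydive chain simulations.
# Vest physics, Control Rig and corrective offsets follow; notably, the right thigh is temporarily rotated around one chain simulation.
# The EventGraph derives movement flags, gravity and vest-variant state.
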AnimBlueprint /BRCosmetics/Characters/Player/Female/Medium/Bodies/F_MED_CinderMax/Meshes/F_MED_CinderMax_AnimBP
target_skeleton=&/F_MED_CinderMax_Skeleton
graphs [
 AnimGraph {nodes [
   "Transform (Modify) Bone - Bone: dyn_chain_06"!(ComponentPose, bAlphaBoolEnabled:bool=True, Translation="0.5,0.8,0", Rotation="0,0,0", Scale="1,1,1") -> (Pose) {
    alphaInputType=Bool
    rotationMode=BMM_Additive
    translationMode=BMM_Additive
    alphaBoolBlend {blendInTime=0.3, blendOption=Linear, blendOutTime=0.3}
    boneToModify {boneName=dyn_chain_06}
    translation=<0.5 0.8>
   }
   n1 = "Constraint - dyn_strap_01_base (PPPPP)"!(
    ComponentPose=Local To Component(LocalPose=Use cached pose 'main_SavedPose'())
    Alpha=1
    ConstraintWeights_0=0.0032
    ConstraintWeights_1:float=0.0051
    ConstraintWeights_2:float=0.5076
    ConstraintWeights_3:float=0.1357
    ConstraintWeights_4:float=0.3485
   ) -> (Pose) {
    alphaBoolBlend {blendOption=Linear}
    boneToModify {boneName=dyn_strap_01_base}
    constraintSetup [
     {targetBone {boneName=deform_glute_l}}
     {targetBone {boneName=thigh_l}}
     {targetBone {boneName=pelvis}}
     {targetBone {boneName=deform_groin_l}}
     {targetBone {boneName=thigh_twist_01_l}}
    ]
    constraintWeights [1,1,1,1,1]
   }
   n2 = "Transform (Modify) Bone - Bone: thigh_r"!(ComponentPose=@n1, bAlphaBoolEnabled:bool=True, Translation="0,0,0", Rotation="0,0,30", Scale="1,1,1") -> (Pose) {
    alphaInputType=Bool
    rotationMode=BMM_Additive
    translationMode=BMM_Additive
    alphaBoolBlend {blendInTime=0.05, blendOption=Linear, blendOutTime=0.8}
    boneToModify {boneName=thigh_r}
    rotation {roll=30}
   }
   n3 = "Transform (Modify) Bone - Bone: thigh_r"(ComponentPose=@n2, bAlphaBoolEnabled=True, Translation="0,0,0", Rotation="-30,0,0", Scale="1,1,1") {
    alphaInputType=Bool
    rotationMode=BMM_Additive
    translationMode=BMM_Additive
    alphaBoolBlend {blendInTime=0.05, blendOption=Linear, blendOutTime=0.8}
    boneToModify {boneName=thigh_r}
    rotation {pitch=-30}
   }
   n4 = RigidBody!(ComponentPose=@n3, Alpha, OverrideWorldGravity:Vector="0,0,-980", ExternalForce) -> (Pose) {
    bOverrideWorldGravity=true
    overridePhysicsAsset=&/F_MED_CinderMax_Physics
    simulationSpace=BaseBoneSpace
    alphaBoolBlend {blendOption=Linear}
    baseBoneRef {boneName=dyn_chain_01}
    simSpaceSettings {worldAlpha=0.05}
   }
   n5 = "Transform (Modify) Bone - Bone: thigh_r"(ComponentPose=@n4, bAlphaBoolEnabled=True, Translation="0,0,0", Rotation="30,0,0", Scale="1,1,1") {
    alphaInputType=Bool
    rotationMode=BMM_Additive
    translationMode=BMM_Additive
    alphaBoolBlend {blendInTime=0.05, blendOption=Linear, blendOutTime=0.8}
    boneToModify {boneName=thigh_r}
    rotation {pitch=30}
   }
   n6 = "Transform (Modify) Bone - Bone: thigh_r"(ComponentPose=@n5, bAlphaBoolEnabled=True, Translation="0,0,0", Rotation="0,0,-30", Scale="1,1,1") {
    alphaInputType=Bool
    rotationMode=BMM_Additive
    translationMode=BMM_Additive
    alphaBoolBlend {blendInTime=0.05, blendOption=Linear, blendOutTime=0.8}
    boneToModify {boneName=thigh_r}
    rotation {roll=-30}
   }
   n7 = Get chain_offset() -> (chain_offset:bool)
   n8 = "Transform (Modify) Bone - Bone: dyn_chain_02"!(ComponentPose=@n6, bAlphaBoolEnabled:bool=@n7, Translation="0,0,0", Rotation="-10.090,2.146,-0.264", Scale="1,1,1") -> (Pose) {
    alphaInputType=Bool
    rotationMode=BMM_Additive
    rotationSpace=BCS_BoneSpace
    translationMode=BMM_Additive
    translationSpace=BCS_BoneSpace
    alphaBoolBlend {blendInTime=0.8, blendOption=Linear, blendOutTime=0.8}
    boneToModify {boneName=dyn_chain_02}
    rotation=<rot -10.09 2.1459 -0.264>
   }
   n9 = "Transform (Modify) Bone - Bone: dyn_chain_03"!(ComponentPose=@n8, bAlphaBoolEnabled:bool=@n7, Translation="0.954,-0.268,-0.239", Rotation="0.017,0.004,-13.096", Scale="1,1,1") -> (Pose) {
    alphaInputType=Bool
    rotationMode=BMM_Additive
    rotationSpace=BCS_BoneSpace
    translationMode=BMM_Additive
    translationSpace=BCS_BoneSpace
    alphaBoolBlend {blendInTime=0.8, blendOption=Linear, blendOutTime=0.8}
    boneToModify {boneName=dyn_chain_03}
    rotation=<rot 0.017 0.004 -13.096>
    translation=<0.954 -0.268 -0.239>
   }
   n10 = "Transform (Modify) Bone - Bone: dyn_chain_04"!(ComponentPose=@n9, bAlphaBoolEnabled:bool=@n7, Translation="0.472,-0.256,-0.206", Rotation="7.237,-10.534,-5.177", Scale="1,1,1") -> (Pose) {
    alphaInputType=Bool
    rotationMode=BMM_Additive
    rotationSpace=BCS_BoneSpace
    translationMode=BMM_Additive
    translationSpace=BCS_BoneSpace
    alphaBoolBlend {blendInTime=0.8, blendOption=Linear, blendOutTime=0.8}
    boneToModify {boneName=dyn_chain_04}
    rotation=<rot 7.237 -10.534 -5.177>
    translation=<0.472 -0.256 -0.206>
   }
   n11 = "Transform (Modify) Bone - Bone: dyn_chain_05"!(ComponentPose=@n10, bAlphaBoolEnabled:bool=@n7, Translation="0.565,0.889,-0.450", Rotation="-21.143,23.196,4.236", Scale="1,1,1") -> (Pose) {
    alphaInputType=Bool
    rotationMode=BMM_Additive
    rotationSpace=BCS_BoneSpace
    translationMode=BMM_Additive
    translationSpace=BCS_BoneSpace
    alphaBoolBlend {blendInTime=0.8, blendOption=Linear, blendOutTime=0.8}
    boneToModify {boneName=dyn_chain_05}
    rotation=<rot -21.143 23.196 4.236>
    translation=<0.565 0.889 -0.45>
   }
   n12 = "Constraint - dyn_chain_06 (PPPPP)"!(
    ComponentPose=@n11
    Alpha=1
    ConstraintWeights_0=0.0141
    ConstraintWeights_1:float=0.0014
    ConstraintWeights_2:float=0.9133
    ConstraintWeights_3:float=0.0002
    ConstraintWeights_4:float=0.0711
   ) -> (Pose) {
    alphaBoolBlend {blendOption=Linear}
    boneToModify {boneName=dyn_chain_06}
    constraintSetup [
     {targetBone {boneName=spine_02}}
     {targetBone {boneName=spine_01}}
     {targetBone {boneName=pelvis}}
     {targetBone {boneName=thigh_twist_01_r}}
     {targetBone {boneName=deform_glute_r}}
    ]
    constraintWeights [1,1,1,1,1]
   }
   n13 = "Constraint - dyn_strap_01_base (PPPPP)"(
    ComponentPose=Local To Component(LocalPose=Use cached pose 'main_SavedPose'())
    Alpha=1
    ConstraintWeights_0=0.0032
    ConstraintWeights_1=0.0051
    ConstraintWeights_2=0.5076
    ConstraintWeights_3=0.1357
    ConstraintWeights_4=0.3485
   ) {
    alphaBoolBlend {blendOption=Linear}
    boneToModify {boneName=dyn_strap_01_base}
    constraintSetup [
     {targetBone {boneName=deform_glute_l}}
     {targetBone {boneName=thigh_l}}
     {targetBone {boneName=pelvis}}
     {targetBone {boneName=deform_groin_l}}
     {targetBone {boneName=thigh_twist_01_l}}
    ]
    constraintWeights [1,1,1,1,1]
   }
   n14 = "Constraint - dyn_chain_06_base (PPPPP)"!(
    ComponentPose=@n13
    Alpha=1
    ConstraintWeights_0=0.0141
    ConstraintWeights_1:float=0.0014
    ConstraintWeights_2:float=0.9133
    ConstraintWeights_3:float=0.0002
    ConstraintWeights_4:float=0.0711
   ) -> (Pose) {
    alphaBoolBlend {blendOption=Linear}
    boneToModify {boneName=dyn_chain_06_base}
    constraintSetup [
     {targetBone {boneName=spine_02}}
     {targetBone {boneName=spine_01}}
     {targetBone {boneName=pelvis}}
     {targetBone {boneName=thigh_twist_01_r}}
     {targetBone {boneName=deform_glute_r}}
    ]
    constraintWeights [1,1,1,1,1]
   }
   n15 = "Transform (Modify) Bone - Bone: dyn_chain_05"(ComponentPose=@n14, bAlphaBoolEnabled=True, Translation="-2.826,0.593,1.133", Rotation="-6.878,-15.014,5.448", Scale="1,1,1") {
    alphaInputType=Bool
    rotationMode=BMM_Additive
    translationMode=BMM_Additive
    alphaBoolBlend {blendOption=Linear}
    boneToModify {boneName=dyn_chain_05}
    rotation=<rot -6.878 -15.014 5.448>
    translation=<-2.826 0.593 1.133>
   }
   n16 = RigidBody(ComponentPose=@n15, OverrideWorldGravity="0,0,-980") {
    bOverrideWorldGravity=true
    overridePhysicsAsset=&/F_MED_CinderMax_Physics_crouching
    simulationSpace=BaseBoneSpace
    alphaBoolBlend {blendOption=Linear}
    baseBoneRef {boneName=dyn_chain_01}
   }
   n17 = "Transform (Modify) Bone - Bone: dyn_chain_01"!(ComponentPose=@n16, bAlphaBoolEnabled:bool=True, Translation="0,0,-0.7", Rotation="-5.000,0.000,-0.000", Scale="1,1,1") -> (Pose) {
    alphaInputType=Bool
    rotationMode=BMM_Additive
    translationMode=BMM_Additive
    alphaBoolBlend {blendOption=Linear}
    boneToModify {boneName=dyn_chain_01}
    rotation {pitch=-5}
    translation {z=-0.7}
   }
   n18 = "Transform (Modify) Bone - Bone: dyn_strap_01"!(ComponentPose=@n17, bAlphaBoolEnabled:bool=True, Translation="0,0,0", Rotation="9.656,-15.221,-2.613", Scale="1,1,1") -> (Pose) {
    alphaInputType=Bool
    rotationMode=BMM_Additive
    translationMode=BMM_Additive
    alphaBoolBlend {blendOption=Linear}
    boneToModify {boneName=dyn_strap_01}
    rotation=<rot 9.656 -15.221 -2.613>
   }
   n19 = Blend Poses by bool(BlendPose_0=Component To Local(ComponentPose=@n18), BlendPose_1=Component To Local(ComponentPose=@n12)) {blendType=Linear}
   n20 = RigidBody(ComponentPose=Local To Component(LocalPose=Use cached pose 'main_SavedPose'()), OverrideWorldGravity="0,0,-300") {
    bOverrideWorldGravity=true
    overridePhysicsAsset=&/F_MED_CinderMax_Physics_skydive
    simulationSpace=BaseBoneSpace
    alphaBoolBlend {blendOption=Linear}
    baseBoneRef {boneName=dyn_chain_01}
   }
   n21 = "Constraint - dyn_chain_05 (PPPPP)"!(
    ComponentPose=@n20
    Alpha=1
    ConstraintWeights_0=0.0024
    ConstraintWeights_1:float=0.0061
    ConstraintWeights_2:float=0.0034
    ConstraintWeights_3:float=0.8
    ConstraintWeights_4:float=0.188
   ) -> (Pose) {
    alphaBoolBlend {blendOption=Linear}
    boneToModify {boneName=dyn_chain_05}
    constraintSetup [
     {targetBone {boneName=thigh_r}}
     {targetBone {boneName=thigh_twist_01_r}}
     {targetBone {boneName=deform_glute_l}}
     {targetBone {boneName=pelvis}}
     {targetBone {boneName=deform_glute_r}}
    ]
    constraintWeights [1,1,1,1,1]
   }
   n22 = "Constraint - dyn_chain_06 (PPPPP)"(
    ComponentPose=@n21
    Alpha=1
    ConstraintWeights_0=0.0141
    ConstraintWeights_1=0.0014
    ConstraintWeights_2=0.9133
    ConstraintWeights_3=0.0002
    ConstraintWeights_4=0.0711
   ) {
    alphaBoolBlend {blendOption=Linear}
    boneToModify {boneName=dyn_chain_06}
    constraintSetup [
     {targetBone {boneName=spine_02}}
     {targetBone {boneName=spine_01}}
     {targetBone {boneName=pelvis}}
     {targetBone {boneName=thigh_twist_01_r}}
     {targetBone {boneName=deform_glute_r}}
    ]
    constraintWeights [1,1,1,1,1]
   }
   n23 = Component To Local(ComponentPose=@n22)
   n24 = "Constraint - dyn_vest_fur_03_l_base (PPPPPPPPPPPPPPPPPPPP)"!(
    ComponentPose=Local To Component(LocalPose=Blend Poses by bool(BlendPose_0=@n23, BlendPose_1=@n19) {blendType=Linear})
    Alpha=1
    ConstraintWeights_0=0
    ConstraintWeights_1:float=0.3114
    ConstraintWeights_2:float=0.0001
    ConstraintWeights_3:float=0.0036
    ConstraintWeights_4:float=0.0001
    ConstraintWeights_5:float=0.1932
    ConstraintWeights_6:float=0.1867
    ConstraintWeights_7:float=0.0001
    ConstraintWeights_8:float=0
    ConstraintWeights_9:float=0
    ConstraintWeights_10:float=0
    ConstraintWeights_11:float=0
    ConstraintWeights_12:float=0
    ConstraintWeights_13:float=0
    ConstraintWeights_14:float=0.1825
    ConstraintWeights_15:float=0
    ConstraintWeights_16:float=0.0012
    ConstraintWeights_17:float=0
    ConstraintWeights_18:float=0.121
    ConstraintWeights_19:float=0
   ) -> (Pose) {
    alphaBoolBlend {blendOption=Linear}
    boneToModify {boneName=dyn_vest_fur_03_l_base}
    constraintSetup [
     {targetBone {boneName=upperarm_twist_02_l}}
     {targetBone {boneName=upperarm_twist_01_l}}
     {targetBone {boneName=spine_02}}
     {targetBone {boneName=spine_04}}
     {targetBone {boneName=spine_03}}
     {targetBone {boneName=spine_05}}
     {targetBone {boneName=clavicle_l}}
     {targetBone {boneName=clavicle_r}}
     {targetBone {boneName=spine_01}}
     {targetBone {boneName=pelvis}}
     {targetBone {boneName=upperarm_twist_01_r}}
     {targetBone {boneName=neck_01}}
     {targetBone {boneName=neck_02}}
     {targetBone {boneName=deform_pec_l}}
     {targetBone {boneName=deform_upperarm_bk_l}}
     {targetBone {boneName=deform_upperarm_bk_r}}
     {targetBone {boneName=deform_upperarm_fr_l}}
     {targetBone {boneName=deform_upperarm_fr_r}}
     {targetBone {boneName=deform_upperarm_up_l}}
     {targetBone {boneName=deform_upperarm_up_r}}
    ]
    constraintWeights [1,1,1,1,1,1,1,1,1,1,1,1,1,1,1,1,1,1,1,1]
   }
   n25 = "Constraint - dyn_vest_fur_02_l_base (PPPPPPPP)"!(
    ComponentPose=@n24
    Alpha=1
    ConstraintWeights_0=0.3618
    ConstraintWeights_1:float=0.0002
    ConstraintWeights_2:float=0.0019
    ConstraintWeights_3:float=0.2886
    ConstraintWeights_4:float=0
    ConstraintWeights_5:float=0.1012
    ConstraintWeights_6:float=0.0374
    ConstraintWeights_7:float=0.2089
   ) -> (Pose) {
    alphaBoolBlend {blendOption=Linear}
    boneToModify {boneName=dyn_vest_fur_02_l_base}
    constraintSetup [
     {targetBone {boneName=upperarm_twist_01_l}}
     {targetBone {boneName=spine_04}}
     {targetBone {boneName=spine_05}}
     {targetBone {boneName=clavicle_l}}
     {targetBone {boneName=deform_pec_l}}
     {targetBone {boneName=deform_upperarm_bk_l}}
     {targetBone {boneName=deform_upperarm_fr_l}}
     {targetBone {boneName=deform_upperarm_up_l}}
    ]
    constraintWeights [1,1,1,1,1,1,1,1]
   }
   n26 = "Constraint - dyn_vest_fur_01_l_base (PPPPPPPP)"!(
    ComponentPose=@n25
    Alpha=1
    ConstraintWeights_0=0.3488
    ConstraintWeights_1:float=0.0281
    ConstraintWeights_2:float=0.0505
    ConstraintWeights_3:float=0.2845
    ConstraintWeights_4:float=0.0094
    ConstraintWeights_5:float=0.0073
    ConstraintWeights_6:float=0.1666
    ConstraintWeights_7:float=0.1048
   ) -> (Pose) {
    alphaBoolBlend {blendOption=Linear}
    boneToModify {boneName=dyn_vest_fur_01_l_base}
    constraintSetup [
     {targetBone {boneName=upperarm_twist_01_l}}
     {targetBone {boneName=spine_04}}
     {targetBone {boneName=spine_05}}
     {targetBone {boneName=clavicle_l}}
     {targetBone {boneName=deform_pec_l}}
     {targetBone {boneName=deform_upperarm_bk_l}}
     {targetBone {boneName=deform_upperarm_fr_l}}
     {targetBone {boneName=deform_upperarm_up_l}}
    ]
    constraintWeights [1,1,1,1,1,1,1,1]
   }
   n27 = "Constraint - dyn_vest_fur_03_r_base (PPPPPPPPPPPPPPPPPPPPPPP)"!(
    ComponentPose=@n26
    Alpha=1
    ConstraintWeights_0=0
    ConstraintWeights_1:float=0.0003
    ConstraintWeights_2:float=0.0096
    ConstraintWeights_3:float=0.0003
    ConstraintWeights_4:float=0.1918
    ConstraintWeights_5:float=0.0007
    ConstraintWeights_6:float=0.1577
    ConstraintWeights_7:float=0
    ConstraintWeights_8:float=0
    ConstraintWeights_9:float=0
    ConstraintWeights_10:float=0.198
    ConstraintWeights_11:float=0
    ConstraintWeights_12:float=0
    ConstraintWeights_13:float=0.0048
    ConstraintWeights_14:float=0
    ConstraintWeights_15:float=0
    ConstraintWeights_16:float=0.1151
    ConstraintWeights_17:float=0.0028
    ConstraintWeights_18:float=0
    ConstraintWeights_19:float=0.0897
    ConstraintWeights_20:float=0.2094
    ConstraintWeights_21:float=0.0002
    ConstraintWeights_22:float=0.0196
   ) -> (Pose) {
    alphaBoolBlend {blendOption=Linear}
    boneToModify {boneName=dyn_vest_fur_03_r_base}
    constraintSetup [
     {targetBone {boneName=upperarm_twist_01_l}}
     {targetBone {boneName=spine_02}}
     {targetBone {boneName=spine_04}}
     {targetBone {boneName=spine_03}}
     {targetBone {boneName=spine_05}}
     {targetBone {boneName=clavicle_l}}
     {targetBone {boneName=clavicle_r}}
     {targetBone {boneName=spine_01}}
     {targetBone {boneName=pelvis}}
     {targetBone {boneName=upperarm_twist_02_r}}
     {targetBone {boneName=upperarm_twist_01_r}}
     {targetBone {boneName=neck_01}}
     {targetBone {boneName=neck_02}}
     {targetBone {boneName=deform_pec_l}}
     {targetBone {boneName=deform_pec_r}}
     {targetBone {boneName=deform_upperarm_bk_l}}
     {targetBone {boneName=deform_upperarm_bk_r}}
     {targetBone {boneName=deform_upperarm_fr_r}}
     {targetBone {boneName=deform_upperarm_up_l}}
     {targetBone {boneName=deform_upperarm_up_r}}
     {targetBone {boneName=extra_vest_back_r}}
     {targetBone {boneName=dyn_vest_fur_02_r}}
     {targetBone {boneName=dyn_vest_fur_03_r}}
    ]
    constraintWeights [1,1,1,1,1,1,1,1,1,1,1,1,1,1,1,1,1,1,1,1,1,1,1]
   }
   n28 = "Constraint - dyn_vest_fur_02_r_base (PPPPPPPPPPPPPPPPPPPPPP)"!(
    ComponentPose=@n27
    Alpha=1
    ConstraintWeights_0=0
    ConstraintWeights_1:float=0
    ConstraintWeights_2:float=0.0004
    ConstraintWeights_3:float=0
    ConstraintWeights_4:float=0.013
    ConstraintWeights_5:float=0
    ConstraintWeights_6:float=0.2829
    ConstraintWeights_7:float=0
    ConstraintWeights_8:float=0
    ConstraintWeights_9:float=0
    ConstraintWeights_10:float=0.3398
    ConstraintWeights_11:float=0
    ConstraintWeights_12:float=0
    ConstraintWeights_13:float=0
    ConstraintWeights_14:float=0
    ConstraintWeights_15:float=0.0959
    ConstraintWeights_16:float=0.0374
    ConstraintWeights_17:float=0
    ConstraintWeights_18:float=0.1966
    ConstraintWeights_19:float=0.015
    ConstraintWeights_20:float=0.0189
    ConstraintWeights_21:float=0
   ) -> (Pose) {
    alphaBoolBlend {blendOption=Linear}
    boneToModify {boneName=dyn_vest_fur_02_r_base}
    constraintSetup [
     {targetBone {boneName=upperarm_twist_01_l}}
     {targetBone {boneName=spine_02}}
     {targetBone {boneName=spine_04}}
     {targetBone {boneName=spine_03}}
     {targetBone {boneName=spine_05}}
     {targetBone {boneName=clavicle_l}}
     {targetBone {boneName=clavicle_r}}
     {targetBone {boneName=spine_01}}
     {targetBone {boneName=pelvis}}
     {targetBone {boneName=upperarm_twist_02_r}}
     {targetBone {boneName=upperarm_twist_01_r}}
     {targetBone {boneName=neck_01}}
     {targetBone {boneName=neck_02}}
     {targetBone {boneName=deform_pec_r}}
     {targetBone {boneName=deform_upperarm_bk_l}}
     {targetBone {boneName=deform_upperarm_bk_r}}
     {targetBone {boneName=deform_upperarm_fr_r}}
     {targetBone {boneName=deform_upperarm_up_l}}
     {targetBone {boneName=deform_upperarm_up_r}}
     {targetBone {boneName=extra_vest_back_r}}
     {targetBone {boneName=dyn_vest_fur_02_r}}
     {targetBone {boneName=dyn_vest_fur_03_r}}
    ]
    constraintWeights [1,1,1,1,1,1,1,1,1,1,1,1,1,1,1,1,1,1,1,1,1,1]
   }
   n29 = "Constraint - dyn_vest_fur_01_r_base (PPPPPPPPPPPPPPPPPPPPPP)"!(
    ComponentPose=@n28
    Alpha=1
    ConstraintWeights_0=0
    ConstraintWeights_1:float=0
    ConstraintWeights_2:float=0.0235
    ConstraintWeights_3:float=0
    ConstraintWeights_4:float=0.0405
    ConstraintWeights_5:float=0
    ConstraintWeights_6:float=0.2377
    ConstraintWeights_7:float=0
    ConstraintWeights_8:float=0
    ConstraintWeights_9:float=0
    ConstraintWeights_10:float=0.2885
    ConstraintWeights_11:float=0
    ConstraintWeights_12:float=0
    ConstraintWeights_13:float=0.0089
    ConstraintWeights_14:float=0
    ConstraintWeights_15:float=0.0088
    ConstraintWeights_16:float=0.1306
    ConstraintWeights_17:float=0
    ConstraintWeights_18:float=0.0927
    ConstraintWeights_19:float=0.0001
    ConstraintWeights_20:float=0.0011
    ConstraintWeights_21:float=0.1676
   ) -> (Pose) {
    alphaBoolBlend {blendOption=Linear}
    boneToModify {boneName=dyn_vest_fur_01_r_base}
    constraintSetup [
     {targetBone {boneName=upperarm_twist_01_l}}
     {targetBone {boneName=spine_02}}
     {targetBone {boneName=spine_04}}
     {targetBone {boneName=spine_03}}
     {targetBone {boneName=spine_05}}
     {targetBone {boneName=clavicle_l}}
     {targetBone {boneName=clavicle_r}}
     {targetBone {boneName=spine_01}}
     {targetBone {boneName=pelvis}}
     {targetBone {boneName=upperarm_twist_02_r}}
     {targetBone {boneName=upperarm_twist_01_r}}
     {targetBone {boneName=neck_01}}
     {targetBone {boneName=neck_02}}
     {targetBone {boneName=deform_pec_r}}
     {targetBone {boneName=deform_upperarm_bk_l}}
     {targetBone {boneName=deform_upperarm_bk_r}}
     {targetBone {boneName=deform_upperarm_fr_r}}
     {targetBone {boneName=deform_upperarm_up_l}}
     {targetBone {boneName=deform_upperarm_up_r}}
     {targetBone {boneName=extra_vest_back_r}}
     {targetBone {boneName=dyn_vest_fur_02_r}}
     {targetBone {boneName=dyn_vest_fur_01_r}}
    ]
    constraintWeights [1,1,1,1,1,1,1,1,1,1,1,1,1,1,1,1,1,1,1,1,1,1]
   }
   n30 = RigidBody(ComponentPose=@n29, OverrideWorldGravity="0,0,-980") {
    bOverrideWorldGravity=true
    overridePhysicsAsset=&/F_MED_CinderMax_Physics_vest
    simulationSpace=BaseBoneSpace
    alphaBoolBlend {blendOption=Linear}
    baseBoneRef {boneName=spine_04}
    simSpaceSettings {worldAlpha=0.1}
   }
   n31 = "Transform (Modify) Bone - Bone: dyn_vest_bottom_fr_l"!(ComponentPose=@n30, bAlphaBoolEnabled:bool=True, Translation="0,0,0", Rotation="0.000,0.000,-30.000", Scale="1,1,1") -> (Pose) {
    alphaInputType=Bool
    rotationMode=BMM_Additive
    translationMode=BMM_Additive
    alphaBoolBlend {blendInTime=0.1, blendOption=Linear, blendOutTime=0.8}
    boneToModify {boneName=dyn_vest_bottom_fr_l}
    rotation {roll=-30}
   }
   n32 = "Transform (Modify) Bone - Bone: dyn_vest_bottom_fr_r"!(ComponentPose=@n31, bAlphaBoolEnabled:bool=True, Translation="0,0,0", Rotation="0.000,0.000,-30.000", Scale="1,1,1") -> (Pose) {
    alphaInputType=Bool
    rotationMode=BMM_Additive
    translationMode=BMM_Additive
    alphaBoolBlend {blendInTime=0.1, blendOption=Linear, blendOutTime=0.8}
    boneToModify {boneName=dyn_vest_bottom_fr_r}
    rotation {roll=-30}
   }
   n33 = "Transform (Modify) Bone - Bone: dyn_vest_bottom_bk_01"!(ComponentPose=@n32, bAlphaBoolEnabled:bool=True, Translation="0,0,0", Rotation="0,0,-15", Scale="1,1,1") -> (Pose) {
    alphaInputType=Bool
    rotationMode=BMM_Additive
    translationMode=BMM_Additive
    alphaBoolBlend {blendInTime=0.1, blendOption=Linear, blendOutTime=0.8}
    boneToModify {boneName=dyn_vest_bottom_bk_01}
    rotation {roll=-15}
   }
   n34 = "Constraint - dyn_zipper_r_MSTR (PPPPPPPPPPP)"!(
    ComponentPose=@n33
    Alpha=1
    ConstraintWeights_0=0
    ConstraintWeights_1:float=0.065
    ConstraintWeights_2:float=0.0535
    ConstraintWeights_3:float=0.5824
    ConstraintWeights_4:float=0.0665
    ConstraintWeights_5:float=0.033
    ConstraintWeights_6:float=0.001
    ConstraintWeights_7:float=0.0003
    ConstraintWeights_8:float=0.0068
    ConstraintWeights_9:float=0.0841
    ConstraintWeights_10:float=0.1074
   ) -> (Pose) {
    alphaBoolBlend {blendOption=Linear}
    boneToModify {boneName=dyn_zipper_r_MSTR}
    constraintSetup [
     {targetBone {boneName=dyn_vest_bottom_side_r}}
     {targetBone {boneName=upperarm_twist_01_r}}
     {targetBone {boneName=spine_03}}
     {targetBone {boneName=spine_04}}
     {targetBone {boneName=spine_02}}
     {targetBone {boneName=deform_pec_r}}
     {targetBone {boneName=deform_upperarm_fr_r}}
     {targetBone {boneName=pelvis}}
     {targetBone {boneName=spine_01}}
     {targetBone {boneName=clavicle_r}}
     {targetBone {boneName=spine_05}}
    ]
    constraintWeights [1,1,1,1,1,1,1,1,1,1,1]
   }
   n35 = "Constraint - dyn_zipper_l_MSTR (PPPPPPPPPPP)"!(
    ComponentPose=@n34
    Alpha=1
    ConstraintWeights_0=0.0535
    ConstraintWeights_1:float=0.001
    ConstraintWeights_2:float=0.5824
    ConstraintWeights_3:float=0.0665
    ConstraintWeights_4:float=0.0003
    ConstraintWeights_5:float=0.0068
    ConstraintWeights_6:float=0.0841
    ConstraintWeights_7:float=0.033
    ConstraintWeights_8:float=0.065
    ConstraintWeights_9:float=0
    ConstraintWeights_10:float=0.1074
   ) -> (Pose) {
    alphaBoolBlend {blendOption=Linear}
    boneToModify {boneName=dyn_zipper_l_MSTR}
    constraintSetup [
     {targetBone {boneName=spine_03}}
     {targetBone {boneName=deform_upperarm_fr_l}}
     {targetBone {boneName=spine_04}}
     {targetBone {boneName=spine_02}}
     {targetBone {boneName=pelvis}}
     {targetBone {boneName=spine_01}}
     {targetBone {boneName=clavicle_l}}
     {targetBone {boneName=deform_pec_l}}
     {targetBone {boneName=upperarm_twist_01_l}}
     {targetBone {boneName=dyn_vest_bottom_side_l}}
     {targetBone {boneName=spine_05}}
    ]
    constraintWeights [1,1,1,1,1,1,1,1,1,1,1]
   }
   n36 = "Transform (Modify) Bone - Bone: attach_backpack"!(ComponentPose=@n35, bAlphaBoolEnabled:bool=Get Vest On() -> (Vest On:bool), Translation="0,0,0", Rotation="0,0,0", Scale="1,1,1") -> (Pose) {
    alphaInputType=Bool
    translationMode=BMM_Additive
    translationSpace=BCS_BoneSpace
    alphaBoolBlend {blendOption=Linear}
    boneToModify {boneName=attach_backpack}
   }
   n37 = "Transform (Modify) Bone - Bone: attach_cape"!(ComponentPose=@n36, bAlphaBoolEnabled:bool=Get Vest On(), Translation="0,-2.5,0", Rotation="0,0,0", Scale="1,1,1") -> (Pose) {
    alphaInputType=Bool
    translationMode=BMM_Additive
    translationSpace=BCS_BoneSpace
    alphaBoolBlend {blendOption=Linear}
    boneToModify {boneName=attach_cape}
    translation {y=-2.5}
   }
   n38 = Control Rig(Source=Component To Local(ComponentPose=@n37)) {
    alphaBoolBlend {blendOption=Linear}
    controlRigAssetReference {blueprintRigClass=&/F_MED_CinderMax_CtrlRig.F_MED_CinderMax_CtrlRig_C}
   }
   n39 = "Transform (Modify) Bone - Bone: dyn_vest_bottom_bk_side_l_base"!(
    ComponentPose=Local To Component(LocalPose=@n38)
    bAlphaBoolEnabled:bool=True
    Translation="0,0,0"
    Rotation="-0.006,12.175,0.006"
    Scale="1,1,1"
   ) -> (Pose) {
    alphaInputType=Bool
    rotationMode=BMM_Additive
    rotationSpace=BCS_BoneSpace
    translationMode=BMM_Additive
    translationSpace=BCS_BoneSpace
    alphaBoolBlend {blendInTime=0.2, blendOption=Linear, blendOutTime=0.2}
    boneToModify {boneName=dyn_vest_bottom_bk_side_l_base}
    rotation=<rot -0.006 12.175 0.006>
   }
   n40 = "Transform (Modify) Bone - Bone: dyn_vest_bottom_bk_base"!(ComponentPose=@n39, bAlphaBoolEnabled:bool=True, Translation="0,0,0", Rotation="-0.006,12.175,0.006", Scale="1,1,1") -> (Pose) {
    alphaInputType=Bool
    rotationMode=BMM_Additive
    rotationSpace=BCS_BoneSpace
    translationMode=BMM_Additive
    translationSpace=BCS_BoneSpace
    alphaBoolBlend {blendInTime=0.2, blendOption=Linear, blendOutTime=0.2}
    boneToModify {boneName=dyn_vest_bottom_bk_base}
    rotation=<rot -0.006 12.175 0.006>
   }
   n41 = "Transform (Modify) Bone - Bone: dyn_vest_bottom_bk_side_r_base"!(ComponentPose=@n40, bAlphaBoolEnabled:bool=True, Translation="0,0,0", Rotation="-0.006,12.175,0.006", Scale="1,1,1") -> (Pose) {
    alphaInputType=Bool
    rotationMode=BMM_Additive
    rotationSpace=BCS_BoneSpace
    translationMode=BMM_Additive
    translationSpace=BCS_BoneSpace
    alphaBoolBlend {blendInTime=0.2, blendOption=Linear, blendOutTime=0.2}
    boneToModify {boneName=dyn_vest_bottom_bk_side_r_base}
    rotation=<rot -0.006 12.175 0.006>
   }
   n42 = "Transform (Modify) Bone - Bone: extra_vest_back_r"!(ComponentPose=@n41, bAlphaBoolEnabled:bool=True, Translation="0,-6,0", Rotation="0.000,0.000,70.000", Scale="1,1,1") -> (Pose) {
    alphaInputType=Bool
    rotationMode=BMM_Additive
    translationMode=BMM_Additive
    alphaBoolBlend {blendInTime=0.3, blendOption=Linear, blendOutTime=0.3}
    boneToModify {boneName=extra_vest_back_r}
    rotation {roll=70}
    translation {y=-6}
   }
   n43 = "Constraint - dyn_chain_04 (P)"(ComponentPose=@n42, Alpha=0.3, ConstraintWeights_0=1) {
    alphaBoolBlend {blendOption=Linear}
    boneToModify {boneName=dyn_chain_04}
    constraintSetup [{targetBone {boneName=pelvis}}]
    constraintWeights [1]
   }
   n44 = "Constraint - dyn_chain_05 (P)"(ComponentPose=@n43, Alpha=1, ConstraintWeights_0=1) {
    alphaBoolBlend {blendOption=Linear}
    boneToModify {boneName=dyn_chain_05}
    constraintSetup [{targetBone {boneName=pelvis}}]
    constraintWeights [1]
   }
   n45 = Look At - Bone: dyn_chain_05(ComponentPose=@n44, Alpha=1) {
    alphaBoolBlend {blendOption=Linear}
    boneToModify {boneName=dyn_chain_05}
    lookAtLocation {x=0}
    lookAtTarget {boneReference {boneName=dyn_chain_04}}
   }
   n46 = "Transform (Modify) Bone - Bone: dyn_chain_05"(ComponentPose=@n45, bAlphaBoolEnabled=True, Translation="0,0,0", Rotation="-10.014,-47.289,-16.568", Scale="1,1,1") {
    alphaInputType=Bool
    rotationMode=BMM_Additive
    rotationSpace=BCS_BoneSpace
    translationMode=BMM_Additive
    translationSpace=BCS_BoneSpace
    alphaBoolBlend {blendOption=Linear}
    boneToModify {boneName=dyn_chain_05}
    rotation=<rot -10.014 -47.289 -16.568>
   }
   n47 = "Constraint - dyn_chain_06 (P)"(ComponentPose=@n46, Alpha=1, ConstraintWeights_0=1) {
    alphaBoolBlend {blendOption=Linear}
    boneToModify {boneName=dyn_chain_06}
    constraintSetup [{targetBone {boneName=pelvis}}]
    constraintWeights [1]
   }
   n48 = "Transform (Modify) Bone - Bone: dyn_chain_03"(ComponentPose=@n47, bAlphaBoolEnabled=True, Translation="-0.5,-0.8,0", Rotation="0,0,0", Scale="1,1,1") {
    alphaInputType=Bool
    rotationMode=BMM_Additive
    translationMode=BMM_Additive
    alphaBoolBlend {blendInTime=0.3, blendOption=Linear, blendOutTime=0.3}
    boneToModify {boneName=dyn_chain_03}
    translation=<-0.5 -0.8>
   }
   Output Pose(Result=Component To Local(ComponentPose=@n48)) {layerGroup=None}
   n49 = Control Rig!(
    Source=Copy Pose From Mesh() {bUseAttachedParent=true}
    Alpha
    deform_value:FN Deform Rig Value Struct Structure=Get deform() -> (deform:FN Deform Rig Value Struct Structure)
   ) -> (Pose) {
    alphaBoolBlend {blendOption=Linear}
    controlRigAssetReference {blueprintRigClass=&/F_MED_CinderMax_DeformRig.F_MED_CinderMax_DeformRig_C}
   }
   n50 = "Constraint - extra_hood_l (PPPPPPPP)"!(
    ComponentPose=Local To Component(LocalPose=@n49)
    Alpha=1
    ConstraintWeights_0=0.0063
    ConstraintWeights_1:float=0.0061
    ConstraintWeights_2:float=0.0044
    ConstraintWeights_3:float=0.4632
    ConstraintWeights_4:float=0.0228
    ConstraintWeights_5:float=0.0217
    ConstraintWeights_6:float=0.4647
    ConstraintWeights_7:float=0.0108
   ) -> (Pose) {
    alphaBoolBlend {blendOption=Linear}
    boneToModify {boneName=extra_hood_l}
    constraintSetup [
     {targetBone {boneName=spine_03}}
     {targetBone {boneName=deform_upperarm_fr_l}}
     {targetBone {boneName=spine_04}}
     {targetBone {boneName=clavicle_l}}
     {targetBone {boneName=upperarm_twist_01_l}}
     {targetBone {boneName=deform_upperarm_up_l}}
     {targetBone {boneName=spine_05}}
     {targetBone {boneName=neck_01}}
    ]
    constraintWeights [1,1,1,1,1,1,1,1]
   }
   n51 = "Constraint - extra_hood_r (PPPPPPPP)"!(
    ComponentPose=@n50
    Alpha=1
    ConstraintWeights_0=0.0203
    ConstraintWeights_1:float=0.007
    ConstraintWeights_2:float=0.0046
    ConstraintWeights_3:float=0.5185
    ConstraintWeights_4:float=0.0212
    ConstraintWeights_5:float=0.0039
    ConstraintWeights_6:float=0.4132
    ConstraintWeights_7:float=0.0113
   ) -> (Pose) {
    alphaBoolBlend {blendOption=Linear}
    boneToModify {boneName=extra_hood_r}
    constraintSetup [
     {targetBone {boneName=upperarm_twist_01_r}}
     {targetBone {boneName=spine_03}}
     {targetBone {boneName=deform_upperarm_fr_r}}
     {targetBone {boneName=clavicle_r}}
     {targetBone {boneName=deform_upperarm_up_r}}
     {targetBone {boneName=deform_upperarm_bk_r}}
     {targetBone {boneName=spine_05}}
     {targetBone {boneName=neck_01}}
    ]
    constraintWeights [1,1,1,1,1,1,1,1]
   }
   n52 = "Transform (Modify) Bone - Bone: attach_backpack"!(ComponentPose=@n51, Alpha:float=1, Translation="-0.000,-2.479,-0.698", Rotation="0,0,0", Scale="1,1,1") -> (Pose) {
    rotationMode=BMM_Additive
    rotationSpace=BCS_BoneSpace
    translationMode=BMM_Additive
    translationSpace=BCS_BoneSpace
    alphaBoolBlend {blendOption=Linear}
    boneToModify {boneName=attach_backpack}
    translation {y=-2.479, z=-0.698}
   }
   n53 = "Constraint - dyn_knee_pad_r (RR)"!(ComponentPose=@n52, Alpha=1, ConstraintWeights_0=1, ConstraintWeights_1:float=1) -> (Pose) {
    alphaBoolBlend {blendOption=Linear}
    boneToModify {boneName=dyn_knee_pad_r}
    constraintSetup [{transformType=Rotation, targetBone {boneName=thigh_r}}, {transformType=Rotation, targetBone {boneName=calf_r}}]
    constraintWeights [1,1]
   }
   n54 = "Constraint - proxy_collar (PPP)"!(ComponentPose=@n53, Alpha=1, ConstraintWeights_0=0.4179, ConstraintWeights_1:float=0.2212, ConstraintWeights_2:float=0.3608) -> (Pose) {
    alphaBoolBlend {blendOption=Linear}
    boneToModify {boneName=proxy_collar}
    constraintSetup [{targetBone {boneName=neck_01}}, {targetBone {boneName=neck_02}}, {targetBone {boneName=spine_05}}]
    constraintWeights [1,1,1]
   }
   n55 = "Constraint - dyn_chain_01_base (PPPPPPPP)"!(
    ComponentPose=@n54
    Alpha=1
    ConstraintWeights_0=0.0067
    ConstraintWeights_1:float=0.651
    ConstraintWeights_2:float=0.0024
    ConstraintWeights_3:float=0.128
    ConstraintWeights_4:float=0.0003
    ConstraintWeights_5:float=0
    ConstraintWeights_6:float=0
    ConstraintWeights_7:float=0.2116
   ) -> (Pose) {
    alphaBoolBlend {blendOption=Linear}
    boneToModify {boneName=dyn_chain_01_base}
    constraintSetup [
     {targetBone {boneName=spine_02}}
     {targetBone {boneName=pelvis}}
     {targetBone {boneName=thigh_twist_01_l}}
     {targetBone {boneName=thigh_twist_01_r}}
     {targetBone {boneName=thigh_r}}
     {targetBone {boneName=deform_glute_r}}
     {targetBone {boneName=deform_groin_l}}
     {targetBone {boneName=deform_groin_r}}
    ]
    constraintWeights [1,1,1,1,1,1,1,1]
   }
   n56 = "Constraint - dyn_chain_06 (PPPPP)"(
    ComponentPose=@n55
    Alpha=1
    ConstraintWeights_0=0.0141
    ConstraintWeights_1=0.0014
    ConstraintWeights_2=0.9133
    ConstraintWeights_3=0.0002
    ConstraintWeights_4=0.0711
   ) {
    alphaBoolBlend {blendOption=Linear}
    boneToModify {boneName=dyn_chain_06}
    constraintSetup [
     {targetBone {boneName=spine_02}}
     {targetBone {boneName=spine_01}}
     {targetBone {boneName=pelvis}}
     {targetBone {boneName=thigh_twist_01_r}}
     {targetBone {boneName=deform_glute_r}}
    ]
    constraintWeights [1,1,1,1,1]
   }
   n57 = "Constraint - dyn_shoe_flaps_r_base (PPPPP)"!(
    ComponentPose=@n56
    Alpha=1
    ConstraintWeights_0=0.3413
    ConstraintWeights_1:float=0.1301
    ConstraintWeights_2:float=0.4073
    ConstraintWeights_3:float=0.0312
    ConstraintWeights_4:float=0.0902
   ) -> (Pose) {
    alphaBoolBlend {blendOption=Linear}
    boneToModify {boneName=dyn_shoe_flaps_r_base}
    constraintSetup [
     {targetBone {boneName=foot_r}}
     {targetBone {boneName=calf_twist_02_r}}
     {targetBone {boneName=calf_twist_01_r}}
     {targetBone {boneName=calf_r}}
     {targetBone {boneName=dyn_shoe_flaps_r}}
    ]
    constraintWeights [1,1,1,1,1]
   }
   n58 = "Constraint - dyn_shoe_flaps_l_base (PPPPP)"!(
    ComponentPose=@n57
    Alpha=1
    ConstraintWeights_0=0.4073
    ConstraintWeights_1:float=0.3413
    ConstraintWeights_2:float=0.1301
    ConstraintWeights_3:float=0.0312
    ConstraintWeights_4:float=0.0902
   ) -> (Pose) {
    alphaBoolBlend {blendOption=Linear}
    boneToModify {boneName=dyn_shoe_flaps_l_base}
    constraintSetup [
     {targetBone {boneName=calf_twist_01_l}}
     {targetBone {boneName=foot_l}}
     {targetBone {boneName=calf_twist_02_l}}
     {targetBone {boneName=calf_l}}
     {targetBone {boneName=dyn_shoe_flaps_l}}
    ]
    constraintWeights [1,1,1,1,1]
   }
   Save cached pose 'main_SavedPose'(Pose=Component To Local(ComponentPose=@n58)) {cacheName=main_SavedPose}
  ]}
 EventGraph {nodes [
   Event BlueprintUpdateAnimation()
   n1 = Get bIsJumping()
   > Set jumping or falling(jumping or falling:bool=@n1) -> (Output_Get:bool)
   n2 = Get bIsPlayingEmote()
   n3 = Get bIsMoving2D()
   > Set chain_offset(chain_offset:bool=NOT Boolean(A=OR Boolean(A=@n3, B=@n2))) -> (Output_Get:bool)
   n4 = Get PawnDeltaRotation() -> (PawnDeltaRotation:Rotator)
   n5 = "float > float"(A="Absolute (Float)"(A=BreakRotator(InRot=@n4).Yaw), B=2)
   n6 = Get bIsMoving2D()
   > Set vest_offset(vest_offset:bool=AND Boolean(A=Get bIsCrouching(), B=OR Boolean(A=@n6, B=@n5))) -> (Output_Get:bool)
   > n7 = Gravity Override State Function DEPRECATED(CharacterPartAnimInstance=Self-Reference(), GameTimeInSeconds=GetGameTimeInSeconds(), GravityOverrideParams=Get Gravity Override Params())
   > n8 = Branch(Condition=Get jumping or falling() -> (jumping or falling:bool))
   > Set Gravity(Gravity:Vector="0,0,-100") -> (Output_Get:Vector)
   Set Gravity(execute=@n8.else, Gravity=@n7)
   Event OnCharacterCustomizationLoaded()
   > Set Vest On(Vest On:bool=CharacterItemDefinitionHasAnyMetaTag(Tags=Make Array("[0]"=(TagName="Cosmetics.Variant.Property.Stage1")))) -> (Output_Get:bool)
   OR Boolean(A=@n1, B=Get bIsFalling())
  ]}
]
variables [
 chain_offset {type=boolean}
 deform {
  type=FN_DeformRig_Value_Struct
  default {
   bicep {location=<-3 4 0>, rotation=<rot 0 0 0>, scale=<1 1 1>}
   elbow_in {location=<0 9 0>, rotation=<rot 0 0 0>, scale=<0.5 1 1>}
   elbow_out {location=<0 -2 0>, rotation=<rot 0 0 0>, scale=<0.75 1 1>}
   glute {location=<-5 0 0>, rotation=<rot -10 -20 0>, scale=<3 1.25 1>}
   groin {location=<-3 12 0>, rotation=<rot 0 0 0>, scale=<0.5 1 1>}
   knee_in {location=<0 -12 0>, rotation=<rot 0 0 0>, scale=<0.5 1 1>}
   knee_out {location=<0 3 0>, rotation=<rot 0 0 0>, scale=<1.5 1 1.5>}
   lat {location=<0 0 10>, rotation=<rot 0 0 0>, scale=<1 1 1>}
   pec {location=<0 2 3>, rotation=<rot 0 0 0>, scale=<1 1 1>}
   upperarm_bk {location=<0 -4 0>, rotation=<rot 0 0 0>, scale=<1 1 1>}
   upperarm_fwd_bk {location=<0 -6 0>, rotation=<rot 0 0 0>, scale=<1 1 1>}
   upperarm_fwd_fr {location=<0 10 0>, rotation=<rot 0 0 0>, scale=<0.5 1 1>}
   upperarm_up {location=<0 0 8>, rotation=<rot 0 0 0>, scale=<0.7 1 1>}
   wrist_dn {location=<0 -6 0>, rotation=<rot 0 0 0>, scale=<0.7 1 1>}
   wrist_up {location=<0 9 0>, rotation=<rot 0 0 0>, scale=<0.7 1 1>}
  }
 }
 Gravity {type=Vector}
 Gravity Override Params {
  type=GravityOverrideParamsStruct
  default {
   skydive_FloatingDown {gravityOverride=<0 -700 100>, windAmplitude=<200 -500 100>, windOffset=<0 -600 100>}
   skydive_GoingUp {gravityOverride=<0 -300 100>, windAmplitude=<200 -500 100>}
  }
 }
 jumping or falling {type=boolean}
 Vest On {type=boolean}
 vest_offset {type=boolean}
]
