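# Control Rig for the F_MED_CinderMax female face accessory that rotates the dyn_tail bones in response to head orientation.
# A spherical pose reader on the head bone outputs a falloff value, which is multiplied by the tail_rot variable.
# That product weights a fixed Euler rotation on dyn_tail_01 and dyn_tail_02, so the default tail_rot of zero gives no effect.
# The curl and facial bones are left untouched, and the pose reader has debug drawing enabled.
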ControlRigBlueprint /BRCosmetics/Characters/Player/Female/Medium/Bodies/F_MED_CinderMax/Meshes/Parts/F_MED_CinderMax_FaceAcc_CtrlRig
graphs [
 RigVMFunctionLibrary {links [], nodes []}
 RigVMModel {nodes [
   RigUnit_BeginExecution = &RigUnit_BeginExecution()
   n1 = &RigUnit_SphericalPoseReader(
    ActiveRegionScaleFactors=(PositiveWidth=1,NegativeWidth=0.5,PositiveHeight=0.75,NegativeHeight=0.75)
    ActiveRegionSize=0.05
    Debug=(bDrawDebug=true,bDraw2D=False,bDrawLocalAxes=False,DebugScale=25,DebugSegments=20,DebugThickness=0.25)
    DriverAxis=(X=0,Y=1,Z=0)
    DriverItem=(Type=Bone,Name="head")
    FalloffRegionScaleFactors=(PositiveWidth=1,NegativeWidth=0.5,PositiveHeight=1,NegativeHeight=1)
    FalloffSize=0.2
    OptionalParentItem=(Type=Bone)
    RotationOffset=(X=105,Y=0,Z=-50)
    ExecutePin=@RigUnit_BeginExecution.ExecutePin
   )
   Multiply = &RigVMFunction_MathFloatMul(A=@n1.OutputParam, B=RigVMVariableNode() {variable=tail_rot}.Value) {template="Multiply::Execute(in A,in B,out Result)"}
   > &RigUnit_ModifyTransforms(
    ItemToModify=((Item=(Type=Bone,Name="dyn_tail_01"),Transform=(Rotation=(X=0,Y=0,Z=0,W=1),Translation=(X=0,Y=0,Z=0),Scale3D=(X=1,Y=1,Z=1))))
    ItemToModify.0.Transform.Rotation=&RigVMFunction_MathQuaternionFromEuler(Euler=(X=0,Y=0,Z=-20)).Result
    Weight=@Multiply.Result
   )
   > &RigUnit_ModifyTransforms(
    ItemToModify=((Item=(Type=Bone,Name="dyn_tail_02"),Transform=(Rotation=(X=0,Y=0,Z=0,W=1),Translation=(X=0,Y=0,Z=0),Scale3D=(X=1,Y=1,Z=1))))
    ItemToModify.0.Transform.Rotation=&RigVMFunction_MathQuaternionFromEuler(Euler=(X=0,Y=0,Z=-20)).Result
    Weight=@Multiply.Result
   )
  ]}
]
hierarchy [
 root {loc=<0 0 0>, rot=<0 0 0 1>}
 pelvis {parent=root, loc=<0 -0.5301 98.0571>, rot=<0.0127 0.707 -0.0127 -0.707>}
 spine_01 {parent=pelvis, loc=<6.1208 -0.9328 0>, rot=<0 0 0.034 -0.9994>}
 spine_02 {parent=spine_01, loc=<6.5996 -0.2095 0>, rot=<0 0 0.0238 -0.9997>}
 spine_03 {parent=spine_02, loc=<6.4914 -0.1237 0>, rot=<0 0 0.0149 -0.9999>}
 spine_04 {parent=spine_03, loc=<6.9598 0.0794 0>, rot=<0 0 -0.0132 -0.9999>}
 spine_05 {parent=spine_04, loc=<6.9987 0 0>, rot=<0 0 -0.0015 -1>}
 neck_01 {parent=spine_05, loc=<11.3138 0.1529 0>, rot=<0 0 -0.2003 -0.9797>}
 neck_02 {parent=neck_01, loc=<4.5627 0 0>, rot=<0 0 0.0504 -0.9987>}
 head {parent=neck_02, loc=<5.3914 0 0>, rot=<0 0 0.115 -0.9934>}
 dyn_curl_01_r {parent=head, loc=<6.2458 2.7955 6.194>, rot=<0.0624 -0.0708 0.9955 0.0073>}
 dyn_curl_02_r {parent=dyn_curl_01_r, loc=<5.0474 -0.0722 0.5958>, rot=<0.9927 0.005 -0.1104 -0.049>}
 dyn_curl_03_r {parent=dyn_curl_02_r, loc=<4.6656 0 0>, rot=<0 0 0 1>}
 dyn_tail_01 {parent=head, loc=<7.5385 -8.194 0>, rot=<0.6548 -0.7558 0 0>}
 dyn_tail_02 {parent=dyn_tail_01, loc=<2.1292 0 0>, rot=<0 0 -0.0487 -0.9988>}
 dyn_tail_03 {parent=dyn_tail_02, loc=<3.4269 -0.3915 0>, rot=<0 0 0.157 -0.9876>}
 dyn_curl_front_01 {parent=head, loc=<8.2729 8.6847 0.7705>, rot=<-0.2841 -0.9571 0.0167 0.0547>}
 dyn_curl_front_02 {parent=dyn_curl_front_01, loc=<1.846 0 0>, rot=<0.0147 0.0272 0.0882 -0.9956>}
 dyn_curl_front_03 {parent=dyn_curl_front_02, loc=<1.8183 0 0>, rot=<0 0 0 1>}
 dyn_curl_01_l {parent=head, loc=<6.2459 2.7955 -6.194>, rot=<-0.0031 -0.9926 -0.0014 -0.1217>}
 dyn_curl_02_l {parent=dyn_curl_01_l, loc=<5.0827 0 0>, rot=<-0.0027 -0.1691 -0.0016 -0.9856>}
 dyn_curl_03_l {parent=dyn_curl_02_l, loc=<4.6657 0 0>, rot=<0 0 0 1>}
 FACIAL_C_FacialRoot {parent=head, loc=<1.9035 -0.2707 0>, rot=<0.5021 -0.4979 -0.5021 -0.4979>}
 FACIAL_C_Nose {parent=FACIAL_C_FacialRoot, loc=<0 -0.1852 9.0725>, rot=<-0.1305 0 0 -0.9914>}
 FACIAL_C_NoseLower {parent=FACIAL_C_Nose, loc=<0 3.2055 -0.3635>, rot=<0.1305 0 0 -0.9914>}
 FACIAL_L_Nostril {parent=FACIAL_C_Nose, loc=<0.9613 2.4459 -0.0988>, rot=<0.0173 0.1305 0.0023 -0.9913>}
 FACIAL_R_Nostril {parent=FACIAL_C_Nose, loc=<-0.9613 2.4459 -0.0988>, rot=<0.0173 -0.1305 -0.0023 -0.9913>}
 FACIAL_C_NoseTip {parent=FACIAL_C_Nose, loc=<0 2.3589 0.688>, rot=<0.0872 0 0 -0.9962>}
 FACIAL_C_LowerLipRotation {parent=FACIAL_C_FacialRoot, loc=<0 2.0753 1.5969>, rot=<0.342 0 0 -0.9397>}
 FACIAL_C_MouthLower {parent=FACIAL_C_LowerLipRotation, loc=<0 -1.8003 5.8593>, rot=<-0.342 0 0 -0.9397>}
 FACIAL_C_LipLower {parent=FACIAL_C_MouthLower, loc=<0 1.1046 2.2333>, rot=<0.0872 0 0 -0.9962>}
 FACIAL_L_LipLower {parent=FACIAL_C_MouthLower, loc=<0.7249 1.1464 2.0657>, rot=<0.0927 -0.1295 0.0273 -0.9869>}
 FACIAL_R_LipLower {parent=FACIAL_C_MouthLower, loc=<-0.7249 1.1464 2.0657>, rot=<0.0927 0.1295 -0.0273 -0.9869>}
 FACIAL_C_Jaw {parent=FACIAL_C_FacialRoot, loc=<0 2.0753 1.5969>, rot=<0.342 0 0 -0.9397>}
 FACIAL_C_Chin {parent=FACIAL_C_Jaw, loc=<0 -0.101 7.7028>, rot=<-0.2588 0 0 -0.9659>}
 FACIAL_L_Ear {parent=FACIAL_C_FacialRoot, loc=<6.2118 -0.3629 -0.5487>, rot=<-0.0262 -0.0007 -0.0262 -0.9993>}
 FACIAL_R_Ear {parent=FACIAL_C_FacialRoot, loc=<-6.2118 -0.3629 -0.5487>, rot=<-0.0262 0.0007 0.0262 -0.9993>}
 clavicle_l {parent=spine_05, loc=<5.4831 -0.1053 -3.2946>, rot=<0.0465 -0.7289 -0.009 -0.6829>}
 upperarm_l {parent=clavicle_l, loc=<9.8658 -1.2136 -0.4105>, rot=<0.0164 -0.3245 -0.0269 -0.9454>}
 clavicle_r {parent=spine_05, loc=<5.4831 -0.1053 3.2946>, rot=<0.7289 0.0465 -0.6829 0.009>}
 upperarm_r {parent=clavicle_r, loc=<-9.8658 1.2136 0.4105>, rot=<0.0164 -0.3245 -0.0269 -0.9454>}
]
variables [tail_rot {type=double, default=0}]
