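# Physics asset for the CinderMax female medium FaceAcc mesh, providing capsule bodies for its dynamic curl and tail bones.
# Left, right and front curls and a tail form four three-bone chains rooted at the head.
# Curls have tight twist and moderate swing limits, while the tail swings wider and its root joint rotates freely.
# Head, upper spine and upper-arm capsules add body volumes, linked by free-rotating constraints, with the left arm capsule thicker than the right.
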PhysicsAsset /BRCosmetics/Characters/Player/Female/Medium/Bodies/F_MED_CinderMax/Meshes/Parts/F_MED_CinderMax_FaceAcc_Physics
bodies {
 dyn_curl_01_l [{shapeType=Capsule, length=1.5137, radius=0.8118, shapeName=dyn_curl_01_l_capsule, center=<2.5233 0 0>, rotation=<rot -90 179.9999 -179.9997>}]
 dyn_curl_01_r [{shapeType=Capsule, length=1.5137, radius=0.8118, shapeName=dyn_curl_01_r_capsule, center=<2.5233 0 0>, rotation=<rot -90 0 -0.0002>}]
 dyn_curl_02_l [{shapeType=Capsule, length=1.5137, radius=0.8118, shapeName=dyn_curl_02_l_capsule, center=<2.5233 0 0>, rotation=<rot -90 -0.0146 0.0149>}]
 dyn_curl_02_r [{shapeType=Capsule, length=1.5137, radius=0.8118, shapeName=dyn_curl_02_r_capsule, center=<2.5233 0 0>, rotation=<rot -90 179.0506 -179.0504>}]
 dyn_curl_03_l [{shapeType=Capsule, length=1.5137, radius=0.8118, shapeName=dyn_curl_03_l_capsule, center=<2.5233 0 0>, rotation=<rot -90 -0.0146 0.0149>}]
 dyn_curl_03_r [{shapeType=Capsule, length=1.5137, radius=0.8118, shapeName=dyn_curl_03_r_capsule, center=<2.5233 0 0>, rotation=<rot -90 180 -179.9998>}]
 dyn_curl_front_01 [{
   shapeType=Capsule
   length=1.5137
   radius=0.8118
   shapeName=dyn_curl_front_01_capsule
   center=<1.692 0 0>
   rotation=<rot 90 179.8418 179.8417>
  }]
 dyn_curl_front_02 [{
   shapeType=Capsule
   length=1.5137
   radius=0.8118
   shapeName=dyn_curl_front_02_capsule
   center=<1.692 0 0>
   rotation=<rot 90 179.9988 179.9987>
  }]
 dyn_curl_front_03 [{shapeType=Capsule, length=1.5137, radius=0.8118, shapeName=dyn_curl_front_03_capsule, center=<1.692 0 0>, rotation=<rot 90 0 -0.0001>}]
 dyn_tail_01 [{shapeType=Capsule, length=1.5137, radius=0.8118, shapeName=dyn_tail_01_capsule, center=<1.797 0 0>, rotation=<rot -90 90 -89.9999>}]
 dyn_tail_02 [{shapeType=Capsule, length=1.5137, radius=0.8118, shapeName=dyn_tail_02_capsule, center=<1.797 0 0>, rotation=<rot -90 179.789 -179.7889>}]
 dyn_tail_03 [{shapeType=Capsule, length=4.0137, radius=2.3118, shapeName=dyn_tail_03_capsule, center=<6.7697 0 0>, rotation=<rot -90 160.5288 -160.5287>}]
 head [
  {shapeType=Capsule, length=8.2237, radius=5.9372, shapeName=head_capsule, center=<0 1.8454 0>, rotation=<rot 90 0.0198 0.0199>}
  {shapeType=Capsule, length=2.8352, radius=4.1016, shapeName=head_capsule, center=<3.963 5.9028 0>}
  {shapeType=Capsule, length=2.6609, radius=2, shapeName=head_capsule, center=<2.7999 -0.0543 -6.5283>, rotation=<rot 90 160.5288 160.5288>}
 ]
 spine_05 [{shapeType=Capsule, length=7.75, radius=7.255, shapeName=spine_05_capsule, center=<4.9509 -8.5287 0>, rotation=<rot 0 -90 0>}]
 upperarm_l [{shapeType=Capsule, length=21.75, radius=8, shapeName=upperarm_l_capsule, center=<12.3273 0 0>, rotation=<rot 90 0 -0.0002>}]
 upperarm_r [{shapeType=Capsule, length=21.75, radius=4, shapeName=upperarm_r_capsule, center=<-12.3273 0 0>, rotation=<rot 90 180 179.9995>}]
}
constraints [
 {bone1Name=dyn_curl_01_l, bone2Name=head, swing1LimitDegrees=15, swing2LimitDegrees=15, twistLimitDegrees=2}
 {bone1Name=dyn_curl_02_l, bone2Name=dyn_curl_01_l, swing1LimitDegrees=30, swing2LimitDegrees=30, twistLimitDegrees=2}
 {bone1Name=dyn_curl_03_l, bone2Name=dyn_curl_02_l, swing1LimitDegrees=30, swing2LimitDegrees=30, twistLimitDegrees=2}
 {bone1Name=dyn_curl_01_r, bone2Name=head, swing1LimitDegrees=15, swing2LimitDegrees=15, twistLimitDegrees=2}
 {bone1Name=dyn_curl_02_r, bone2Name=dyn_curl_01_r, swing1LimitDegrees=30, swing2LimitDegrees=30, twistLimitDegrees=2}
 {bone1Name=dyn_curl_03_r, bone2Name=dyn_curl_02_r, swing1LimitDegrees=30, swing2LimitDegrees=30, twistLimitDegrees=2}
 {bone1Name=dyn_curl_front_01, bone2Name=head, swing1LimitDegrees=30, swing2LimitDegrees=30, twistLimitDegrees=2}
 {bone1Name=dyn_curl_front_02, bone2Name=dyn_curl_front_01, swing1LimitDegrees=30, swing2LimitDegrees=30, twistLimitDegrees=2}
 {bone1Name=dyn_curl_front_03, bone2Name=dyn_curl_front_02, swing1LimitDegrees=30, swing2LimitDegrees=30, twistLimitDegrees=2}
 {
  bone1Name=dyn_tail_01
  bone2Name=head
  swing1LimitDegrees=60
  swing1Motion=Free
  swing2LimitDegrees=30
  swing2Motion=Free
  twistLimitDegrees=15
  twistMotion=Free
 }
 {bone1Name=dyn_tail_02, bone2Name=dyn_tail_01, swing1LimitDegrees=60, swing2LimitDegrees=30, twistLimitDegrees=15}
 {bone1Name=dyn_tail_03, bone2Name=dyn_tail_02, swing1LimitDegrees=60, swing2LimitDegrees=30, twistLimitDegrees=15}
 {
  bone1Name=head
  bone2Name=spine_05
  swing1LimitDegrees=45
  swing1Motion=Free
  swing2LimitDegrees=45
  swing2Motion=Free
  twistLimitDegrees=45
  twistMotion=Free
 }
 {
  bone1Name=upperarm_r
  bone2Name=spine_05
  swing1LimitDegrees=45
  swing1Motion=Free
  swing2LimitDegrees=45
  swing2Motion=Free
  twistLimitDegrees=45
  twistMotion=Free
 }
 {
  bone1Name=upperarm_l
  bone2Name=spine_05
  swing1LimitDegrees=45
  swing1Motion=Free
  swing2LimitDegrees=45
  swing2Motion=Free
  twistLimitDegrees=45
  twistMotion=Free
 }
]
