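# Post-process style AnimBlueprint for the CinderMax face accessory, adding secondary motion to the copied character pose.
# It blends upper-arm branches from the second custom mesh part over the attached parent's pose.
# Two head-space RigidBody nodes with separate physics assets, one front-end, feed a Control Rig driven by tail_rot.
# Update logic sets front-end/normal alphas and gravity, easing tail_rot up while crouched, targeting and turning.
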AnimBlueprint /BRCosmetics/Characters/Player/Female/Medium/Bodies/F_MED_CinderMax/Meshes/Parts/F_MED_CinderMax_FaceAcc_AnimBP
target_skeleton=&/F_MED_CinderMax_FaceAcc_Skeleton
graphs [
 AnimGraph {nodes [
   n1 = Copy Pose From Mesh(SourceMeshComponent=Get BodyMesh() -> (BodyMesh:Skeletal Mesh Component Object Reference)) {bUseAttachedParent=true}
   n2 = Copy Pose From Mesh() {bUseAttachedParent=true}
   n3 = RigidBody!(
    ComponentPose=Local To Component(LocalPose=Layered blend per bone(BasePose=@n2, BlendPoses_0=@n1) {layerSetup [{branchFilters [{boneName=upperarm_l}, {boneName=upperarm_r}]}]})
    Alpha
    OverrideWorldGravity:Vector="0,0,-980"
    ExternalForce
   ) -> (Pose) {
    bOverrideWorldGravity=true
    overridePhysicsAsset=&/F_MED_CinderMax_FaceAcc_Physics
    simulationSpace=BaseBoneSpace
    alphaBoolBlend {blendOption=Linear}
    baseBoneRef {boneName=head}
    componentAppliedLinearAccClamp=<2500 2500 2500>
    componentLinearAccScale=<0.35 0.35 0.35>
    componentLinearVelScale=<0.35 0.35 0.35>
    simSpaceSettings {maxAngularAcceleration=10, maxAngularVelocity=10, maxLinearAcceleration=10, maxLinearVelocity=10, worldAlpha=1}
   }
   n4 = RigidBody(ComponentPose=@n3, OverrideWorldGravity="0,0,-980") {
    bOverrideWorldGravity=true
    overridePhysicsAsset=&/F_MED_CinderMax_FaceAcc_FE_Physics
    simulationSpace=BaseBoneSpace
    alphaBoolBlend {blendOption=Linear}
    baseBoneRef {boneName=head}
    componentAppliedLinearAccClamp=<2500 2500 2500>
    componentLinearAccScale=<0.35 0.35 0.35>
    componentLinearVelScale=<0.35 0.35 0.35>
    simSpaceSettings {maxAngularAcceleration=10, maxAngularVelocity=10, maxLinearAcceleration=10, maxLinearVelocity=10, worldAlpha=1}
   }
   n5 = Control Rig!(Source=Component To Local(ComponentPose=@n4), Alpha, tail_rot:double=Get tail_rot() -> (tail_rot:double)) -> (Pose) {
    alphaBoolBlend {blendOption=Linear}
    controlRigAssetReference {blueprintRigClass=&/F_MED_CinderMax_FaceAcc_CtrlRig.F_MED_CinderMax_FaceAcc_CtrlRig_C}
   }
   Output Pose(Result=@n5) {layerGroup=None}
  ]}
 EventGraph {nodes [
   n1 = Event OnCharacterCustomizationLoaded()
   > Set BodyMesh(BodyMesh="Get (a copy)"(Array:Array of Skeletal Mesh Component Budgeted Object References=@n1.CustomMeshParts, Dimension 1=1) -> (Output:Skeletal Mesh Component Budgeted Object Reference))
   n2 = Event BlueprintUpdateAnimation()
   > n3 = Branch(Condition=Get bIsInFrontEnd())
   > Set FE Alpha(FE Alpha:double=1) -> (Output_Get:double)
   > n4 = Set Normal Alpha(Normal Alpha:double=0) -> (Output_Get:double)
   n5 = Gravity Override State Function DEPRECATED(
    execute=[@n8.then, @n4.then]
    CharacterPartAnimInstance=Self-Reference()
    GameTimeInSeconds=GetGameTimeInSeconds()
    GravityOverrideParams=Get Gravity Override Params()
   )
   > Set Gravity(Gravity:Vector=@n5) -> (Output_Get:Vector)
   n6 = "Absolute (Float)"(A=BreakRotator(InRot=Get PawnDeltaRotation() -> (PawnDeltaRotation:Rotator)).Yaw)
   > n7 = Branch(Condition=AND Boolean!(A=Get bIsCrouching(), B=Get bIsTargeting(), C:bool="float > float"(A=@n6, B=2)) -> (ReturnValue))
   > Set tail_rot(tail_rot:double=FInterpTo_Constant!(Current:double=Get tail_rot(), Target:double=1, DeltaTime:double=@n2.DeltaTimeX, InterpSpeed:double=3) -> (ReturnValue:double)) -> (Output_Get:double)
   Set tail_rot(execute=@n7.else, tail_rot=FInterpTo_Constant(Current=Get tail_rot(), Target=0, DeltaTime=@n2.DeltaTimeX, InterpSpeed=3))
   Set FE Alpha(execute=@n3.else, FE Alpha=0)
   > n8 = Set Normal Alpha(Normal Alpha=1)
  ]}
]
variables [
 BodyMesh {type=SkeletalMeshComponent}
 FE Alpha {type=number}
 Gravity {type=Vector}
 Gravity Override Params {type=GravityOverrideParamsStruct}
 Normal Alpha {type=number}
 tail_rot {type=number}
]
